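# Physics asset for the Ms Alpine backpack cosmetic, defining simulated bodies for its dangling ropes and zipper pulls.
# A capsule on the pack bone anchors box bodies on the dyn_ bones.
# Left and right zippers hang from the pack, the right as a two-link chain, while a tightly limited rope root branches into up, down and a two-link rope chain.
# Angular swing and twist limits on each joint constrain how far the pieces move.
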PhysicsAsset /BRCosmetics/Accessories/FORT_Backpacks/Backpack_F_MED_Ms_Alpine/Meshes/F_MED_Ms_Alpine_Dyn
bodies {
 dyn_rope_01 [{shapeType=Box, extentX=5, extentY=3.1118, extentZ=1, center=<2.5 0 0>}]
 dyn_rope_02 [{shapeType=Box, extentX=5, extentY=5, extentZ=1, center=<2.5 0 0>}]
 dyn_ropes_dn [{shapeType=Box, extentX=8, extentY=5, extentZ=5, center=<3 0 0>}]
 dyn_ropes_root [{shapeType=Box, extentX=7.5, extentY=6, extentZ=5, center=<1 0 0>}]
 dyn_ropes_up [{shapeType=Box, extentX=8, extentY=6, extentZ=5, center=<3 0 0>}]
 dyn_zipper_l_01 [{shapeType=Box, extentX=4.4906, extentY=1.6903, extentZ=1, center=<2 0 0>}]
 dyn_zipper_r_01 [{shapeType=Box, extentX=4.4906, extentY=1.6903, extentZ=1, center=<2 0 0>}]
 dyn_zipper_r_02 [{shapeType=Box, extentX=1.5, extentY=1.5, extentZ=0.5, center=<1 0 0>}]
 pack [{shapeType=Capsule, length=6.4924, radius=10.44, center=<-0.5721 -8.0064 -10.3848>}]
}
constraints [
 {bone1Name=dyn_zipper_r_01, bone2Name=pack, swing1LimitDegrees=45, swing2LimitDegrees=20, twistLimitDegrees=2}
 {bone1Name=dyn_zipper_r_02, bone2Name=dyn_zipper_r_01, swing1LimitDegrees=45, swing2LimitDegrees=20, twistLimitDegrees=2}
 {bone1Name=dyn_zipper_l_01, bone2Name=pack, swing1LimitDegrees=45, swing2LimitDegrees=20, twistLimitDegrees=2}
 {bone1Name=dyn_ropes_root, bone2Name=pack, swing1LimitDegrees=10, swing2LimitDegrees=2, twistLimitDegrees=2}
 {bone1Name=dyn_ropes_up, bone2Name=dyn_ropes_root, swing1LimitDegrees=45, swing2LimitDegrees=20, twistLimitDegrees=2}
 {bone1Name=dyn_ropes_dn, bone2Name=dyn_ropes_root, swing1LimitDegrees=30, swing2LimitDegrees=20, twistLimitDegrees=2}
 {bone1Name=dyn_rope_01, bone2Name=dyn_ropes_root, swing1LimitDegrees=20, swing2LimitDegrees=45, twistLimitDegrees=5}
 {bone1Name=dyn_rope_02, bone2Name=dyn_rope_01, swing1LimitDegrees=20, swing2LimitDegrees=45, twistLimitDegrees=5}
]
